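# Niagara cosmetic effect for the PaleSpooky player body: small dark goop drops that drift upward from skeletal mesh sockets, leaving oily trails.
# FN_Drops uses two spawn groups to pick the socket index, both stopped by User.bisHandHidden, and renders noise-pushed, shrinking mesh and sprite particles.
# The drops send location events, which presumably feed FN_Oil_Trails, a short-lived black tapering ribbon emitter whose own spawn rate is zero.
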NiagaraSystem /BRCosmetics/Characters/Player/Male/Medium/Bodies/M_MED_Palespooky/FX/NS_PaleSpooky_GoopDrops_
dependencies {
 usedDataInterfaces [&NiagaraDataInterfaceCurve, &NiagaraDataInterfaceSkeletalMesh]
 usedDynamicInputs []
 usedModules [
  &self:SystemSpawnScript.SystemScriptSource.SystemScriptGraph.NiagaraNodeAssignment_0.SetVariables_#1
  &SystemState
 ]
}
emitters {
 FN_Drops {
  emitterUpdateScript [
   &EmitterState(Life Cycle Mode=::NewEnumerator1, Loop Duration=1, UseLoopDelay=false)
   &SpawnRate(
    SpawnRate=&MakeCustomFloatFromBool(Bool=$User.bisHandHidden, True Float=0, False Float=&RandomRangeFloat(Minimum=1, Maximum=1.5))
    Use Spawn Probability=false
   )
   &SpawnRate(
    SpawnGroup=1
    SpawnRate=&MakeCustomFloatFromBool(Bool=$User.bisHandHidden, True Float=0, False Float=&RandomRangeFloat(Minimum=1, Maximum=1.5))
    Use Spawn Probability=false
   ) {module=SpawnRate001}
  ]
  particleSpawnScript [
   /Niagara/Modules/Spawn/Initialization/InitializeParticle(
    Color=<rgba 0.015 0.015 0.015 1>
    Lifetime=&UniformRangedFloat(Minimum=0.8, Maximum=0.4)
    Mass=&UniformRangedFloat(Minimum=0.6, Maximum=1.2)
    Mesh Scale=&RandomRangeVector(Minimum=<0.8 0.2 0.2>, Maximum=<1 0.4 0.4>)
    Module.Write Color=true
    Module.Write Lifetime=true
    Module.Write Mass=true
    Module.Write Material Random=true
    Module.Write Position=false
    Module.Write Scale=true
    Module.Write SpriteRotation=true
    Module.Write SpriteSize=true
    Position=&SimulationPosition()
    Sprite Rotation=&NormalizedAngleToDegrees(NormalizedAngle=&UniformRangedFloat(Minimum=0, Maximum=1))
    Sprite Size=<4 20>
   )
   /Niagara/Modules/Spawn/Location/SphereLocation(MaskSpawnGroup=false, Module.Override Randomness=false, Sphere Radius=3, Surface Only=false)
   &SkeletalMeshLocation(
    Bone / Socket Index=&SpawnGroup()
    Bone Sampling Mode=::NewEnumerator4
    Inherit Velocity=true
    "Inherit Velocity (Scale)"=0.02
    MaskSpawnGroup=false
    Skeletal Mesh=$System.Skeletal Mesh
    VelocitySpeedLimit=false
   ) {module=SkeletalMeshLocation001}
   &CylinderLocation(
    Cylinder Height=320
    MaskSpawnGroup=false
    Module.Enable Lathe Profile=false
    Module.Override Randomness=false
    Module.Surface Only=false
    Offset=<0 0 300>
   ) {enabled=false}
   &AddVelocity(Scale Added Velocity=<1 1 10>, Velocity=<0 0 5>)
  ]
  particleUpdateScript [
   &ParticleState()
   &AccelerationForce(Acceleration=&RandomRangeVector(Minimum=<0 0 50>, Maximum=<1 1 100>))
   &AddVelocity(Coordinate Space=::World, Velocity=&RandomRangeVector(Minimum=<0.5 0.5 0>, Maximum=<-0.5 -0.5 0>)) {module=AddVelocity001}
   &VectorNoiseForce(Force Amount=&RandomRangeFloat(Minimum=300, Maximum=600), Module.Use Random Noise Time Interval=false)
   &Color() {enabled=false}
   &Drag(Drag=3, Use Linear Drag=true, Use Rotational Drag=false)
   &SolveForcesAndVelocity(Clamp Velocity=false, Limit Acceleration=false)
   &GenerateLocationEvent(Module.Use Event Send Rate=true)
   &ScaleMeshSize(Scale Factor=&VectorFromFloat(Value=&FloatFromCurve(FloatCurve=DataInterface() {propertyValues {Curve {keys [<at 0.5 1>,<at 1 0>]}}}, Scale Curve=1.2)))
  ]
  rendererClasses [&NiagaraMeshRendererProperties, &NiagaraSpriteRendererProperties]
  renderers [
   {rendererClass=&NiagaraMeshRendererProperties, rendererIndex=0}
   {rendererClass=&NiagaraSpriteRendererProperties, rendererIndex=1}
  ]
 }
 FN_Oil_Trails {
  emitterUpdateScript [
   &EmitterScalability(
    Module.bApplySpawnCountScale=false
    Module.bApplySpawnCountScaleByDistanceFraction=false
    Module.bUseMaxDistance=false
    Module.bUseMinDistance=false
   ) {enabled=false}
   &EmitterLifeCycle() {enabled=false}
   &EmitterState()
   &SpawnRate(SpawnRate=0, Use Spawn Probability=false)
   &SpawnPerUnit(Module.Use Max Movement Threshold=true, Module.Use Movement Tolerance=false, Use Spawn Probability=false) {enabled=false}
  ]
  particleSpawnScript [&InitializeRibbon(
    Color=<rgba 0 0 0 1>
    Lifetime=&RandomRangeFloat(Minimum=0.35, Maximum=0.25)
    Module.Write Color=true
    Module.Write Lifetime=true
    Module.Write Mass=true
    Module.Write Material Random=true
    Module.Write Position=true
    Module.Write RibbonTwist=false
    Module.Write RibbonWidth=true
    Position:Vector3f=&Add_Vector(A=/Niagara/DynamicInputs/Helpers/SimulationPosition(), B=<0 0 0>)
    Ribbon Width=1.5
   )]
  particleUpdateScript [
   &ParticleState()
   &AccelerationForce()
   &SolveForcesAndVelocity(Clamp Velocity=false, Limit Acceleration=false)
   &Color(
    Color=$Particles.Initial.Color
    Scale Alpha=&FloatFromCurve(FloatCurve=DataInterface() {propertyValues {Curve {keys [<at 0 2>,<at 1 1>]}}}, Scale Curve=2)
   )
   &ScaleRibbonWidth(Ribbon Width Scale=&FloatFromCurve(FloatCurve=DataInterface() {propertyValues {Curve {keys [<at 0 3>,<at 0.1823 1.5698>,<at 0.6547 0.3179>,<at 1 0>]}}}))
  ]
  rendererClasses [&NiagaraRibbonRendererProperties]
  renderers [{rendererClass=&NiagaraRibbonRendererProperties}]
 }
}
summary {
 systemName=NS_PaleSpooky_GoopDrops_
 emitters [
  {emitterName=FN_Drops, rendererClasses [&NiagaraMeshRendererProperties, &NiagaraSpriteRendererProperties]}
  {emitterName=FN_Oil_Trails, rendererClasses [&NiagaraRibbonRendererProperties]}
 ]
 userVariables [
  User.bisHandHidden {type=NiagaraBool}
  User.RibbonIDOffset {type=NiagaraInt32}
 ]
}
