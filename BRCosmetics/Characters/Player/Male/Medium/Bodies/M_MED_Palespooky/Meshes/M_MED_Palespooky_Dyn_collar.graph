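# Physics asset for the dynamic collar on the Palespooky male medium player body, with capsules on the upper spine, clavicles and upper arms.
# Nine two-segment collar chains hang from tiny MSTR root capsules: a–c on each clavicle, d left/right and the back on spine_05.
# Root and body joints have free angular motion, while chain joints have tight swing and twist limits.
# Side segments use boxes; back segments use capsules.
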PhysicsAsset /BRCosmetics/Characters/Player/Male/Medium/Bodies/M_MED_Palespooky/Meshes/M_MED_Palespooky_Dyn_collar
bodies {
 clavicle_l [{shapeType=Capsule, length=8.25, radius=6.25, center=<10.3401 -0.0502 1.8786>, rotation=<rot 79.9176 0 0>}]
 clavicle_r [{shapeType=Capsule, length=8.25, radius=6.25, center=<-10.2636 0 -1.5929>, rotation=<rot 74.1084 0 0>}]
 dyn_collar_01_bk [{shapeType=Capsule, length=4.2636, radius=1.9121, center=<5.6142 0.0099 0.4604>, rotation=<rot 0 94.0695 -84.9679>}]
 dyn_collar_02_bk [{shapeType=Capsule, length=3.7175, radius=1.8689, center=<4.8534 0.1085 -0.0626>, rotation=<rot 0 -86.7081 -89.5261>}]
 dyn_collar_a_01_l [{shapeType=Box, extentX=8, extentY=5.5, extentZ=2, center=<3.9413 0 0.0003>}]
 dyn_collar_a_01_r [{shapeType=Box, extentX=8, extentY=6, extentZ=2, center=<4.4961 0.0001 0>}]
 dyn_collar_a_02_l [{shapeType=Box, extentX=8, extentY=4, extentZ=2, center=<4.3646 -0.0001 -0.5947>, rotation=<rot -10 0 0>}]
 dyn_collar_a_02_r [{shapeType=Box, extentX=8, extentY=5, extentZ=2, center=<4.276 0 -0.0004>}]
 dyn_collar_a_l_MSTR [{shapeType=Capsule, length=0.5, radius=0.5}]
 dyn_collar_a_r_MSTR [{shapeType=Capsule, length=0.5, radius=0.5}]
 dyn_collar_b_01_l [{shapeType=Box, extentX=9, extentY=7, extentZ=3, center=<4.749 0 -0.0005>}]
 dyn_collar_b_01_r [{shapeType=Box, extentX=8.5, extentY=8, extentZ=3.5, center=<4.8021 -0.0001 0>}]
 dyn_collar_b_02_l [{shapeType=Box, extentX=8, extentY=6.5, extentZ=2, center=<4.9636 -0.0002 0>}]
 dyn_collar_b_02_r [{shapeType=Box, extentX=9.5, extentY=7.5, extentZ=3, center=<4.9166 0.0002 -2.167>, rotation=<rot -19.9999 0 0>}]
 dyn_collar_b_l_MSTR [{shapeType=Capsule, length=0.5, radius=0.5}]
 dyn_collar_b_r_MSTR [{shapeType=Capsule, length=0.5, radius=0.5}]
 dyn_collar_bk_MSTR [{shapeType=Capsule, length=0.5, radius=0.5}]
 dyn_collar_c_01_l [{shapeType=Box, extentX=9.5, extentY=6, extentZ=2, center=<5.0381 0 0>}]
 dyn_collar_c_01_r [{shapeType=Box, extentX=9, extentY=8, extentZ=2, center=<5.1172 0 -0.0001>}]
 dyn_collar_c_02_l [{shapeType=Box, extentX=10, extentY=4, extentZ=2, center=<5.0093 0 0.0001>}]
 dyn_collar_c_02_r [{shapeType=Box, extentX=8, extentY=6.5, extentZ=2, center=<4.6739 -0.0001 0.3093>}]
 dyn_collar_c_l_MSTR [{shapeType=Capsule, length=0.5, radius=0.5}]
 dyn_collar_c_r_MSTR [{shapeType=Capsule, length=0.5, radius=0.5}]
 dyn_collar_d_01_l [{shapeType=Box, extentX=8, extentY=6, extentZ=2, center=<4.2287 0 0>}]
 dyn_collar_d_01_r [{shapeType=Box, extentX=8, extentY=8, extentZ=2, center=<3.9672 0 0.0001>}]
 dyn_collar_d_02_l [{shapeType=Box, extentX=6.5, extentY=3.5, extentZ=2, center=<3.0968 0 0>}]
 dyn_collar_d_02_r [{shapeType=Box, extentX=8, extentY=6.5, extentZ=2, center=<3.5596 0 -0.0001>}]
 dyn_collar_d_l_MSTR [{shapeType=Capsule, length=0.5, radius=0.5}]
 dyn_collar_d_r_MSTR [{shapeType=Capsule, length=0.5, radius=0.5}]
 spine_04 [{shapeType=Capsule, length=7.75, radius=11, center=<4.3051 -0.0002 0>}]
 spine_05 [{shapeType=Capsule, length=11, radius=9.5, center=<1.4246 0.336 0>}]
 upperarm_l [{shapeType=Capsule, length=10.75, radius=4.5, center=<9.4245 -1.0457 -1>, rotation=<rot 90 0 0>}]
 upperarm_r [{shapeType=Capsule, length=10.75, radius=4.5, center=<-9.4245 1.0893 1>, rotation=<rot 90 0 0>}]
}
constraints [
 {
  bone1Name=dyn_collar_bk_MSTR
  bone2Name=spine_05
  swing1LimitDegrees=45
  swing1Motion=Free
  swing2LimitDegrees=45
  swing2Motion=Free
  twistLimitDegrees=45
  twistMotion=Free
 }
 {
  bone1Name=dyn_collar_d_l_MSTR
  bone2Name=spine_05
  swing1LimitDegrees=45
  swing1Motion=Free
  swing2LimitDegrees=45
  swing2Motion=Free
  twistLimitDegrees=45
  twistMotion=Free
 }
 {bone1Name=dyn_collar_d_01_l, bone2Name=dyn_collar_d_l_MSTR, swing1LimitDegrees=5, swing2LimitDegrees=15, twistLimitDegrees=2}
 {bone1Name=dyn_collar_d_02_l, bone2Name=dyn_collar_d_01_l, swing1LimitDegrees=10, swing2LimitDegrees=15, twistLimitDegrees=2}
 {
  bone1Name=dyn_collar_d_r_MSTR
  bone2Name=spine_05
  swing1LimitDegrees=45
  swing1Motion=Free
  swing2LimitDegrees=45
  swing2Motion=Free
  twistLimitDegrees=45
  twistMotion=Free
 }
 {bone1Name=dyn_collar_d_01_r, bone2Name=dyn_collar_d_r_MSTR, swing1LimitDegrees=5, swing2LimitDegrees=15, twistLimitDegrees=2}
 {bone1Name=dyn_collar_d_02_r, bone2Name=dyn_collar_d_01_r, swing1LimitDegrees=5, swing2LimitDegrees=15, twistLimitDegrees=2}
 {
  bone1Name=dyn_collar_a_l_MSTR
  bone2Name=clavicle_l
  swing1LimitDegrees=45
  swing1Motion=Free
  swing2LimitDegrees=45
  swing2Motion=Free
  twistLimitDegrees=45
  twistMotion=Free
 }
 {bone1Name=dyn_collar_a_01_l, bone2Name=dyn_collar_a_l_MSTR, swing1LimitDegrees=5, swing2LimitDegrees=15, twistLimitDegrees=2}
 {bone1Name=dyn_collar_a_02_l, bone2Name=dyn_collar_a_01_l, swing1LimitDegrees=10, swing2LimitDegrees=15, twistLimitDegrees=2}
 {
  bone1Name=dyn_collar_b_l_MSTR
  bone2Name=clavicle_l
  swing1LimitDegrees=45
  swing1Motion=Free
  swing2LimitDegrees=45
  swing2Motion=Free
  twistLimitDegrees=45
  twistMotion=Free
 }
 {bone1Name=dyn_collar_b_01_l, bone2Name=dyn_collar_b_l_MSTR, swing1LimitDegrees=5, swing2LimitDegrees=15, twistLimitDegrees=2}
 {bone1Name=dyn_collar_b_02_l, bone2Name=dyn_collar_b_01_l, swing1LimitDegrees=10, swing2LimitDegrees=15, twistLimitDegrees=2}
 {
  bone1Name=dyn_collar_c_l_MSTR
  bone2Name=clavicle_l
  swing1LimitDegrees=45
  swing1Motion=Free
  swing2LimitDegrees=45
  swing2Motion=Free
  twistLimitDegrees=45
  twistMotion=Free
 }
 {bone1Name=dyn_collar_c_01_l, bone2Name=dyn_collar_c_l_MSTR, swing1LimitDegrees=5, swing2LimitDegrees=15, twistLimitDegrees=2}
 {bone1Name=dyn_collar_c_02_l, bone2Name=dyn_collar_c_01_l, swing1LimitDegrees=10, swing2LimitDegrees=15, twistLimitDegrees=2}
 {
  bone1Name=dyn_collar_a_r_MSTR
  bone2Name=clavicle_r
  swing1LimitDegrees=45
  swing1Motion=Free
  swing2LimitDegrees=45
  swing2Motion=Free
  twistLimitDegrees=45
  twistMotion=Free
 }
 {bone1Name=dyn_collar_a_01_r, bone2Name=dyn_collar_a_r_MSTR, swing1LimitDegrees=5, swing2LimitDegrees=15, twistLimitDegrees=2}
 {bone1Name=dyn_collar_a_02_r, bone2Name=dyn_collar_a_01_r, swing1LimitDegrees=10, swing2LimitDegrees=15, twistLimitDegrees=2}
 {
  bone1Name=dyn_collar_b_r_MSTR
  bone2Name=clavicle_r
  swing1LimitDegrees=45
  swing1Motion=Free
  swing2LimitDegrees=45
  swing2Motion=Free
  twistLimitDegrees=45
  twistMotion=Free
 }
 {bone1Name=dyn_collar_b_01_r, bone2Name=dyn_collar_b_r_MSTR, swing1LimitDegrees=5, swing2LimitDegrees=15, twistLimitDegrees=2}
 {bone1Name=dyn_collar_b_02_r, bone2Name=dyn_collar_b_01_r, swing1LimitDegrees=5, swing2LimitDegrees=15, twistLimitDegrees=2}
 {
  bone1Name=dyn_collar_c_r_MSTR
  bone2Name=clavicle_r
  swing1LimitDegrees=45
  swing1Motion=Free
  swing2LimitDegrees=45
  swing2Motion=Free
  twistLimitDegrees=45
  twistMotion=Free
 }
 {bone1Name=dyn_collar_c_01_r, bone2Name=dyn_collar_c_r_MSTR, swing1LimitDegrees=5, swing2LimitDegrees=15, twistLimitDegrees=2}
 {bone1Name=dyn_collar_c_02_r, bone2Name=dyn_collar_c_01_r, swing1LimitDegrees=5, swing2LimitDegrees=15, twistLimitDegrees=2}
 {
  bone1Name=clavicle_l
  bone2Name=spine_05
  swing1LimitDegrees=45
  swing1Motion=Free
  swing2LimitDegrees=45
  swing2Motion=Free
  twistLimitDegrees=45
  twistMotion=Free
 }
 {
  bone1Name=clavicle_r
  bone2Name=spine_05
  swing1LimitDegrees=45
  swing1Motion=Free
  swing2LimitDegrees=45
  swing2Motion=Free
  twistLimitDegrees=45
  twistMotion=Free
 }
 {
  bone1Name=upperarm_r
  bone2Name=clavicle_r
  swing1LimitDegrees=45
  swing1Motion=Free
  swing2LimitDegrees=45
  swing2Motion=Free
  twistLimitDegrees=45
  twistMotion=Free
 }
 {
  bone1Name=upperarm_l
  bone2Name=clavicle_l
  swing1LimitDegrees=45
  swing1Motion=Free
  swing2LimitDegrees=45
  swing2Motion=Free
  twistLimitDegrees=45
  twistMotion=Free
 }
 {
  bone1Name=spine_05
  bone2Name=spine_04
  swing1LimitDegrees=45
  swing1Motion=Free
  swing2LimitDegrees=45
  swing2Motion=Free
  twistLimitDegrees=45
  twistMotion=Free
 }
 {bone1Name=dyn_collar_01_bk, bone2Name=dyn_collar_bk_MSTR, swing1LimitDegrees=5, swing2LimitDegrees=15, twistLimitDegrees=2}
 {bone1Name=dyn_collar_02_bk, bone2Name=dyn_collar_01_bk, swing1LimitDegrees=5, swing2LimitDegrees=15, twistLimitDegrees=2}
]
