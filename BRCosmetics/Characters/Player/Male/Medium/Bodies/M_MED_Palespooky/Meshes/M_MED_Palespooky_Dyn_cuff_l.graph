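# Physics asset for the left sleeve cuff of the Palespooky male medium character body, simulating secondary cuff motion.
# Four dynamic cuff bones carry capsule shapes, each constrained to the lower-arm twist bone with identical limited swing and tight twist.
# Capsules on the left hand and forearm twist bone provide bodies for the cuff to collide against.
# The fourth cuff bone and the hand each use two capsules to approximate their shape.
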PhysicsAsset /BRCosmetics/Characters/Player/Male/Medium/Bodies/M_MED_Palespooky/Meshes/M_MED_Palespooky_Dyn_cuff_l
bodies {
 dyn_cuff_01_l [{shapeType=Capsule, length=3.2251, radius=1.2015, center=<3.7597 -0.8959 0>, rotation=<rot 0 -15.0039 -100.0002>}]
 dyn_cuff_02_l [{shapeType=Capsule, length=1.8803, radius=1.5282, center=<2.7147 0.0902 -0.1354>, rotation=<rot -0.9881 87.1842 -78.6298>}]
 dyn_cuff_03_l [{shapeType=Capsule, length=2.6239, radius=1.6913, center=<4.4326 0.4077 -0.0359>, rotation=<rot 0.8588 -80.202 -87.6393>}]
 dyn_cuff_04_l [
  {shapeType=Capsule, length=2.522, radius=1.2477, center=<3.8492 2.8843 -0.2686>, rotation=<rot 0 116.02 -76.1268>}
  {shapeType=Capsule, length=2.022, radius=0.9977, center=<4.0453 -0.9105 0.7921>, rotation=<rot 0 116.0173 -61.1264>}
 ]
 hand_l [
  {shapeType=Capsule, length=4.75, radius=2.25, center=<4.5244 -0.1953 -2.0381>, rotation=<rot 84.9999 -0.0001 0>}
  {shapeType=Capsule, length=4.75, radius=2.25, center=<4.4721 -0.1953 2.3902>, rotation=<rot 79.9896 180 180>}
 ]
 lowerarm_twist_01_l [{shapeType=Capsule, length=3.75, radius=3.25, center=<0 0.0001 0.642>, rotation=<rot 90 0 0>}]
}
constraints [
 {bone1Name=dyn_cuff_01_l, bone2Name=lowerarm_twist_01_l, swing1LimitDegrees=10, swing2LimitDegrees=30, twistLimitDegrees=5}
 {bone1Name=dyn_cuff_02_l, bone2Name=lowerarm_twist_01_l, swing1LimitDegrees=10, swing2LimitDegrees=30, twistLimitDegrees=5}
 {bone1Name=dyn_cuff_03_l, bone2Name=lowerarm_twist_01_l, swing1LimitDegrees=10, swing2LimitDegrees=30, twistLimitDegrees=5}
 {bone1Name=dyn_cuff_04_l, bone2Name=lowerarm_twist_01_l, swing1LimitDegrees=10, swing2LimitDegrees=30, twistLimitDegrees=5}
]
