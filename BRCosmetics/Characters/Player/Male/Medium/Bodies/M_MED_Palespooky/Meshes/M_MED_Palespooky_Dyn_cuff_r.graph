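# Physics asset for the dynamic right sleeve cuff of the medium male Palespooky player body cosmetic.
# Four cuff bones each carry a capsule and are constrained directly to the right lower-arm twist bone, not to each other.
# All four constraints share tight, asymmetric swing limits and a narrow twist limit.
# Capsules on the hand and lower-arm twist bone supply the surrounding arm shapes.
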PhysicsAsset /BRCosmetics/Characters/Player/Male/Medium/Bodies/M_MED_Palespooky/Meshes/M_MED_Palespooky_Dyn_cuff_r
bodies {
 dyn_cuff_01_r [{shapeType=Capsule, length=1.6235, radius=1.3809, center=<4.7273 0.0986 -0.3719>, rotation=<rot 0 98.0045 -86.9969>}]
 dyn_cuff_02_r [{shapeType=Capsule, length=2.3386, radius=1.395, center=<5.6366 0 0>, rotation=<rot 0 86.3761 -85.9894>}]
 dyn_cuff_03_r [{shapeType=Capsule, length=0.9101, radius=1.3345, center=<3.6391 0.0691 -0.3381>, rotation=<rot 0 -71.9904 -88.1962>}]
 dyn_cuff_04_r [{shapeType=Capsule, length=1.9924, radius=1.4588, center=<5.4852 0.3192 1.6647>, rotation=<rot 0 85.9721 -58.1157>}]
 hand_r [
  {shapeType=Capsule, length=4.75, radius=2.25, center=<-4.4721 0.2 -2.3902>, rotation=<rot 79.9896 180 180>}
  {shapeType=Capsule, length=4.75, radius=2.25, center=<-4.7836 0.1999 2.3349>, rotation=<rot 90 0.0198 0.0198>}
 ]
 lowerarm_twist_01_r [{shapeType=Capsule, length=3.75, radius=3.25, center=<0 0.0001 -0.642>, rotation=<rot 90 0 0>}]
}
constraints [
 {bone1Name=dyn_cuff_01_r, bone2Name=lowerarm_twist_01_r, swing1LimitDegrees=10, swing2LimitDegrees=30, twistLimitDegrees=5}
 {bone1Name=dyn_cuff_02_r, bone2Name=lowerarm_twist_01_r, swing1LimitDegrees=10, swing2LimitDegrees=30, twistLimitDegrees=5}
 {bone1Name=dyn_cuff_03_r, bone2Name=lowerarm_twist_01_r, swing1LimitDegrees=10, swing2LimitDegrees=30, twistLimitDegrees=5}
 {bone1Name=dyn_cuff_04_r, bone2Name=lowerarm_twist_01_r, swing1LimitDegrees=10, swing2LimitDegrees=30, twistLimitDegrees=5}
]
