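# Physics asset for the Palespooky medium male player body's dynamic jacket, defining its bodies and joints.
# Six two-segment flap chains (front, side and back per side) hang from master bodies joined to spine_01, with tight twist limits.
# Front and side flaps use thin boxes, back flaps use capsules, and pelvis, spine and thigh capsules represent the body.
# Extra constraints tie front flaps to side flaps, asymmetrically between left and right.
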PhysicsAsset /BRCosmetics/Characters/Player/Male/Medium/Bodies/M_MED_Palespooky/Meshes/M_MED_Palespooky_Dyn_jacket
bodies {
 dyn_jacket_l_01 [{shapeType=Box, extentX=8, extentY=14.5, extentZ=2, center=<5 0 0>, rotation=<rot 0 0 -10.001>}]
 dyn_jacket_l_02 [{shapeType=Box, extentX=10, extentY=10.5, extentZ=2, center=<5 0 0>, rotation=<rot 1.8652 10.638 -9.8336>}]
 dyn_jacket_l_MSTR [{shapeType=Capsule, length=1, radius=1, rotation=<rot 90 0 0>}]
 dyn_jacket_l_bk_01 [{shapeType=Capsule, length=4, radius=4.5, shapeName=dyn_jacket_r_bk_01_capsule, center=<5 0 2.5>, rotation=<rot 0 0 85.0001>}]
 dyn_jacket_l_bk_02 [{shapeType=Capsule, length=4, radius=4.5, shapeName=dyn_jacket_r_bk_02_capsule, center=<3 0 2>, rotation=<rot 0 0 90>}]
 dyn_jacket_l_bk_MSTR [{shapeType=Capsule, length=1, radius=1, rotation=<rot 90 0 0>}]
 dyn_jacket_l_fr_01 [{shapeType=Box, extentX=9.5, extentY=11.5, extentZ=2, center=<7.0148 0.5 -0.0001>, rotation=<rot 0 -0.0001 -2.3029>}]
 dyn_jacket_l_fr_02 [{shapeType=Box, extentX=5.5, extentY=9, extentZ=2, center=<3.5269 1.8134 0.0291>, rotation=<rot -3.5712 26.5271 -0.5406>}]
 dyn_jacket_l_fr_MSTR [{shapeType=Capsule, length=1, radius=1, rotation=<rot 90 0 0>}]
 dyn_jacket_r_01 [{shapeType=Box, extentX=10, extentY=14, extentZ=2, center=<5 0 0>, rotation=<rot 0 0 10.0005>}]
 dyn_jacket_r_02 [{shapeType=Box, extentX=10, extentY=13.5, extentZ=2, center=<5 0 0>, rotation=<rot 0 0 10.0005>}]
 dyn_jacket_r_MSTR [{shapeType=Capsule, length=1, radius=1, rotation=<rot 90 0 0>}]
 dyn_jacket_r_bk_01 [{shapeType=Capsule, length=4, radius=4.5, shapeName=dyn_jacket_r_bk_01_capsule, center=<5 0 2.5>, rotation=<rot 0 0 85.0001>}]
 dyn_jacket_r_bk_02 [{shapeType=Capsule, length=4, radius=4.5, shapeName=dyn_jacket_r_bk_02_capsule, center=<3 0 2>, rotation=<rot 0 0 90>}]
 dyn_jacket_r_bk_MSTR [{shapeType=Capsule, length=1, radius=1, rotation=<rot 90 0 0>}]
 dyn_jacket_r_fr_01 [{shapeType=Box, extentX=10, extentY=11, extentZ=2, center=<4.4936 1.3899 0>, rotation=<rot 0 15.0037 15.0019>}]
 dyn_jacket_r_fr_02 [{shapeType=Box, extentX=12, extentY=5.5, extentZ=2, center=<6.7037 -0.5 0>}]
 dyn_jacket_r_fr_MSTR [{shapeType=Capsule, length=1, radius=1, rotation=<rot 90 0 0>}]
 pelvis [
  {shapeType=Capsule, length=10.0739, radius=9.2271, center=<-0.9177 -0.8108 0>}
  {shapeType=Capsule, length=11.5936, radius=8.1762, center=<-7.049 -1.2079 0>}
 ]
 spine_01 [{shapeType=Capsule, length=6.5198, radius=8.6282, center=<-0.1959 0.0546 0>}]
 thigh_twist_01_l [{shapeType=Capsule, length=18.5, radius=7, center=<5.036 0.7144 0.2922>, rotation=<rot 88.3441 -179.9997 172.3883>}]
 thigh_twist_01_r [{shapeType=Capsule, length=19.5, radius=7, center=<-5.0732 -0.4999 -0.5209>, rotation=<rot 86.8184 -180 174.9994>}]
}
constraints [
 {bone1Name=spine_01, bone2Name=pelvis, swing1LimitDegrees=45, swing2LimitDegrees=45, twistLimitDegrees=45}
 {bone1Name=dyn_jacket_r_bk_01, bone2Name=dyn_jacket_r_bk_MSTR, swing1LimitDegrees=15, swing2LimitDegrees=45, twistLimitDegrees=2}
 {bone1Name=dyn_jacket_r_bk_02, bone2Name=dyn_jacket_r_bk_01, swing1LimitDegrees=15, swing2LimitDegrees=45, twistLimitDegrees=2}
 {bone1Name=dyn_jacket_l_bk_01, bone2Name=dyn_jacket_l_bk_MSTR, swing1LimitDegrees=15, swing2LimitDegrees=45, twistLimitDegrees=2}
 {bone1Name=dyn_jacket_l_bk_02, bone2Name=dyn_jacket_l_bk_01, swing1LimitDegrees=15, swing2LimitDegrees=45, twistLimitDegrees=2}
 {bone1Name=dyn_jacket_l_fr_01, bone2Name=dyn_jacket_l_fr_MSTR, swing1LimitDegrees=8, swing2LimitDegrees=60, twistLimitDegrees=5}
 {bone1Name=dyn_jacket_l_fr_02, bone2Name=dyn_jacket_l_fr_01, swing1LimitDegrees=8, swing2LimitDegrees=60, twistLimitDegrees=5}
 {bone1Name=dyn_jacket_r_fr_01, bone2Name=dyn_jacket_r_fr_MSTR, swing1LimitDegrees=15, swing2LimitDegrees=45, twistLimitDegrees=2}
 {bone1Name=dyn_jacket_r_fr_02, bone2Name=dyn_jacket_r_fr_01, swing1LimitDegrees=15, swing2LimitDegrees=45, twistLimitDegrees=2}
 {bone1Name=dyn_jacket_l_02, bone2Name=dyn_jacket_l_01, swing1LimitDegrees=10, swing2LimitDegrees=60, twistLimitDegrees=2}
 {bone1Name=dyn_jacket_l_01, bone2Name=dyn_jacket_l_MSTR, swing1LimitDegrees=30, swing2LimitDegrees=60, twistLimitDegrees=2}
 {bone1Name=dyn_jacket_r_01, bone2Name=dyn_jacket_r_MSTR, swing1LimitDegrees=30, swing2LimitDegrees=60, twistLimitDegrees=2}
 {bone1Name=dyn_jacket_r_02, bone2Name=dyn_jacket_r_01, swing1LimitDegrees=30, swing2LimitDegrees=60, twistLimitDegrees=2}
 {
  bone1Name=thigh_twist_01_l
  bone2Name=pelvis
  swing1LimitDegrees=45
  swing1Motion=Free
  swing2LimitDegrees=45
  swing2Motion=Free
  twistLimitDegrees=45
  twistMotion=Free
 }
 {
  bone1Name=thigh_twist_01_r
  bone2Name=pelvis
  swing1LimitDegrees=45
  swing1Motion=Free
  swing2LimitDegrees=45
  swing2Motion=Free
  twistLimitDegrees=45
  twistMotion=Free
 }
 {
  bone1Name=dyn_jacket_r_bk_MSTR
  bone2Name=spine_01
  swing1LimitDegrees=45
  swing1Motion=Free
  swing2LimitDegrees=45
  swing2Motion=Free
  twistLimitDegrees=45
  twistMotion=Free
 }
 {
  bone1Name=dyn_jacket_l_bk_MSTR
  bone2Name=spine_01
  swing1LimitDegrees=45
  swing1Motion=Free
  swing2LimitDegrees=45
  swing2Motion=Free
  twistLimitDegrees=45
  twistMotion=Free
 }
 {
  bone1Name=dyn_jacket_l_fr_MSTR
  bone2Name=spine_01
  swing1LimitDegrees=45
  swing1Motion=Free
  swing2LimitDegrees=45
  swing2Motion=Free
  twistLimitDegrees=45
  twistMotion=Free
 }
 {
  bone1Name=dyn_jacket_r_fr_MSTR
  bone2Name=spine_01
  swing1LimitDegrees=45
  swing1Motion=Free
  swing2LimitDegrees=45
  swing2Motion=Free
  twistLimitDegrees=45
  twistMotion=Free
 }
 {
  bone1Name=dyn_jacket_l_MSTR
  bone2Name=spine_01
  swing1LimitDegrees=45
  swing1Motion=Free
  swing2LimitDegrees=45
  swing2Motion=Free
  twistLimitDegrees=45
  twistMotion=Free
 }
 {
  bone1Name=dyn_jacket_r_MSTR
  bone2Name=spine_01
  swing1LimitDegrees=45
  swing1Motion=Free
  swing2LimitDegrees=45
  swing2Motion=Free
  twistLimitDegrees=45
  twistMotion=Free
 }
 {
  bone1Name=dyn_jacket_l_01
  bone2Name=dyn_jacket_l_fr_01
  swing1LimitDegrees=10
  swing2LimitDegrees=15
  swing2Motion=Free
  twistLimitDegrees=45
  twistMotion=Free
 }
 {
  bone1Name=dyn_jacket_r_02
  bone2Name=dyn_jacket_r_fr_01
  swing1LimitDegrees=10
  swing2LimitDegrees=15
  swing2Motion=Free
  twistLimitDegrees=45
  twistMotion=Free
 }
]
