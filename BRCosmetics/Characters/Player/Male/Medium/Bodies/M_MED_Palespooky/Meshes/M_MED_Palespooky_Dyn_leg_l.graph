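# Physics asset for the left lower leg of the Palespooky male medium player body cosmetic, set up for dynamic trouser-leg bones.
# Capsule bodies cover the calf twist bone, the foot and three trouser-leg bones.
# Each trouser bone is constrained to the calf twist bone with identical, asymmetric swing and tight twist limits.
# The foot capsule has no constraint.
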PhysicsAsset /BRCosmetics/Characters/Player/Male/Medium/Bodies/M_MED_Palespooky/Meshes/M_MED_Palespooky_Dyn_leg_l
bodies {
 calf_twist_01_l [{shapeType=Capsule, length=4.75, radius=2.75, center=<-4.1357 0.0001 0.0146>, rotation=<rot 84.9999 0 0>}]
 dyn_trouserLegs_01_l [{shapeType=Capsule, length=1.5079, radius=1.2087, center=<4.6961 0.4645 -0.289>, rotation=<rot 0 -78.9569 -82.9749>}]
 dyn_trouserLegs_02_l [{shapeType=Capsule, length=1.8974, radius=1.0577, center=<4.8729 -0.2299 -0.0352>, rotation=<rot 0 85.7011 -83.2287>}]
 dyn_trouserLegs_03_l [{shapeType=Capsule, length=1.9039, radius=1.2428, center=<4.5638 0.9716 0.3539>, rotation=<rot 0 93.9071 -71.9234>}]
 foot_l [{shapeType=Capsule, length=12.5, radius=3.25, center=<-5.1793 3.9179 -0.3>, rotation=<rot 90 90 164.9949>}]
}
constraints [
 {bone1Name=dyn_trouserLegs_01_l, bone2Name=calf_twist_01_l, swing1LimitDegrees=15, swing2LimitDegrees=45, twistLimitDegrees=10}
 {bone1Name=dyn_trouserLegs_02_l, bone2Name=calf_twist_01_l, swing1LimitDegrees=15, swing2LimitDegrees=45, twistLimitDegrees=10}
 {bone1Name=dyn_trouserLegs_03_l, bone2Name=calf_twist_01_l, swing1LimitDegrees=15, swing2LimitDegrees=45, twistLimitDegrees=10}
]
