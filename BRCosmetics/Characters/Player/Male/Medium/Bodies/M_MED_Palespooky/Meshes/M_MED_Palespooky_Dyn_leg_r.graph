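# Physics asset for the right trouser leg of the male medium Palespooky cosmetic body, likely driving secondary cloth-like motion on its dynamic bones.
# Four small capsules on the dyn_trouserLegs bones are each constrained directly to calf_twist_01_r with identical swing and twist limits, rather than chained to one another.
# Larger capsules on calf_twist_01_r and foot_r cover the lower leg and foot, with the foot left unconstrained, apparently as a collider for the trouser bones.
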PhysicsAsset /BRCosmetics/Characters/Player/Male/Medium/Bodies/M_MED_Palespooky/Meshes/M_MED_Palespooky_Dyn_leg_r
bodies {
 calf_twist_01_r [{shapeType=Capsule, length=4, radius=3.5, center=<4.1046 0.0002 -0.3726>, rotation=<rot 90 0 0>}]
 dyn_trouserLegs_01_r [{shapeType=Capsule, length=2.182, radius=1.2574, center=<3.6534 -0.2998 -0.3289>, rotation=<rot 0 -77.4064 -79.1592>}]
 dyn_trouserLegs_02_r [{shapeType=Capsule, length=1.045, radius=0.75, center=<4.2648 -0.0468 -0.2433>, rotation=<rot 0 73.9635 -87.2998>}]
 dyn_trouserLegs_03_r [{shapeType=Capsule, length=1.8595, radius=1.1392, center=<6.5897 -1.5792 -0.3031>, rotation=<rot 0 -119.5067 -89.8495>}]
 dyn_trouserLegs_04_r [{shapeType=Capsule, length=1.5181, radius=1.3517, center=<3.2734 0.1476 -0.1791>, rotation=<rot 0 110.4555 -80.6692>}]
 foot_r [{shapeType=Capsule, length=12.5, radius=3.25, center=<4.9742 -4.5339 0.3333>, rotation=<rot 90 51.3402 126.3352>}]
}
constraints [
 {bone1Name=dyn_trouserLegs_01_r, bone2Name=calf_twist_01_r, swing1LimitDegrees=15, swing2LimitDegrees=45, twistLimitDegrees=10}
 {bone1Name=dyn_trouserLegs_02_r, bone2Name=calf_twist_01_r, swing1LimitDegrees=15, swing2LimitDegrees=45, twistLimitDegrees=10}
 {bone1Name=dyn_trouserLegs_03_r, bone2Name=calf_twist_01_r, swing1LimitDegrees=15, swing2LimitDegrees=45, twistLimitDegrees=10}
 {bone1Name=dyn_trouserLegs_04_r, bone2Name=calf_twist_01_r, swing1LimitDegrees=15, swing2LimitDegrees=45, twistLimitDegrees=10}
]
