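# Physics asset for the dynamic hair of the Palespooky medium male player cosmetic body, covering top, side and back hair bones.
# Each hair bone has a capsule; top and side strands form chains of up to three bones rooted at the head, while back bones attach directly to it.
# Tight swing and very small twist limits constrain the hair, with wider swing on some back bones.
# Head and upper-spine capsules supply body shapes, joined by a free-rotating constraint.
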PhysicsAsset /BRCosmetics/Characters/Player/Male/Medium/Bodies/M_MED_Palespooky/Meshes/Parts/M_MED_Palespooky_Dyn_hair
bodies {
 dyn_hair_01_bk [{shapeType=Capsule, length=2, radius=1, center=<2.1529 0 0>, rotation=<rot 90 0 0>}]
 dyn_hair_02_bk [{shapeType=Capsule, length=2, radius=1, center=<3.5615 0 0>, rotation=<rot 90 0 0>}]
 dyn_hair_03_bk [{shapeType=Capsule, length=2, radius=1, center=<4.5534 0 0>, rotation=<rot 90 0 0>}]
 dyn_hair_04_bk [{shapeType=Capsule, length=2, radius=1, center=<4.5133 0 0.4663>, rotation=<rot 74.9929 -180 -180>}]
 dyn_hair_05_bk [{shapeType=Capsule, length=2, radius=1, center=<4.2242 0 0>, rotation=<rot 90 0 0>}]
 dyn_hair_a_01_l [{shapeType=Capsule, length=1.2816, radius=0.8392, center=<3.5449 0 0>, rotation=<rot 90 0 0>}]
 dyn_hair_a_01_r [{shapeType=Capsule, length=0.4154, radius=1.7968, center=<1.2292 0 0>, rotation=<rot 84.9895 -180 -180>}]
 dyn_hair_a_02_l [{shapeType=Capsule, length=1.9584, radius=0.9635, center=<2.4443 0 0>, rotation=<rot 90 0 0>}]
 dyn_hair_a_02_r [{shapeType=Capsule, length=1.6646, radius=1.5075, center=<1.6386 0 0>, rotation=<rot 90 0 0>}]
 dyn_hair_a_03_l [{shapeType=Capsule, length=1.6309, radius=0.6428, center=<1.8885 0 0>, rotation=<rot 90 0 0>}]
 dyn_hair_a_03_r [{shapeType=Capsule, length=2.4877, radius=0.6737, center=<1.8666 0 0>, rotation=<rot 90 0 0>}]
 dyn_hair_b_01_l [{shapeType=Capsule, length=1, radius=1.2615, center=<4.324 -0.0001 0>, rotation=<rot 90 0 0>}]
 dyn_hair_b_01_r [{shapeType=Capsule, length=2.3068, radius=1.3078, center=<2.564 0 0>, rotation=<rot 90 0 0>}]
 dyn_hair_b_02_l [{shapeType=Capsule, length=2.561, radius=0.7917, center=<2.714 0 0>, rotation=<rot 84.9936 -180 -180>}]
 dyn_hair_b_02_r [{shapeType=Capsule, length=2.4651, radius=1.1779, center=<2.2531 -0.4845 -0.0001>, rotation=<rot 90 0 25.0017>}]
 dyn_hair_b_03_r [{shapeType=Capsule, length=3.0444, radius=0.6052, center=<2.6782 0 0>, rotation=<rot 90 0 0>}]
 dyn_hair_c_01_l [{shapeType=Capsule, length=3.3212, radius=1.0223, center=<3.9918 0.0001 -0.1141>, rotation=<rot 84.9776 -180 180>}]
 dyn_hair_c_01_r [{shapeType=Capsule, length=2.9209, radius=1.766, center=<2.8905 -0.0001 0>, rotation=<rot 90 0 0>}]
 dyn_hair_c_02_r [{shapeType=Capsule, length=2.8309, radius=0.581, center=<2.5679 0 0>, rotation=<rot 90 0 0>}]
 dyn_topHair_a_01_l [{shapeType=Capsule, length=0.9951, radius=1.7478, center=<2.2966 0 -0.33>, rotation=<rot 80 0 0>}]
 dyn_topHair_a_01_r [{shapeType=Capsule, length=1.6939, radius=1.65, center=<2.208 0 0>, rotation=<rot 84.9999 0 0>}]
 dyn_topHair_a_02_l [{shapeType=Capsule, length=2.8627, radius=1.1563, center=<2 0 0>, rotation=<rot 74.9997 0 0>}]
 dyn_topHair_a_02_r [{shapeType=Capsule, length=2.8352, radius=1.0024, center=<2.5009 0 0>, rotation=<rot 90 0 0>}]
 dyn_topHair_b_01_l [{shapeType=Capsule, length=1.0389, radius=2.1749, center=<3.9901 0 0>, rotation=<rot 90 0 0>}]
 dyn_topHair_b_01_r [{shapeType=Capsule, length=2.1101, radius=1.2213, center=<3.5029 0 0.2776>, rotation=<rot 79.9999 0 0>}]
 dyn_topHair_b_02_l [{shapeType=Capsule, length=2.8905, radius=1.357, center=<3.2621 0 0.9356>, rotation=<rot 59.9967 -180 -180>}]
 dyn_topHair_b_02_r [{shapeType=Capsule, length=3.1351, radius=1.1056, center=<3.0545 0 0>, rotation=<rot 90 0 0>}]
 dyn_topHair_c_01_l [{shapeType=Capsule, length=1.4799, radius=2.0375, center=<3.5155 0 0>, rotation=<rot 79.9999 0 0>}]
 dyn_topHair_c_01_r [{shapeType=Capsule, length=1.1592, radius=1.4837, center=<3.79 0 0.1843>, rotation=<rot 79.9999 0 0>}]
 dyn_topHair_c_02_l [{shapeType=Capsule, length=2.7455, radius=0.7987, center=<2.944 0 0>, rotation=<rot 90 0 0>}]
 dyn_topHair_c_02_r [{shapeType=Capsule, length=2.7836, radius=0.9038, center=<2.752 0 0>, rotation=<rot 90 0 0>}]
 dyn_topHair_d_01_l [{shapeType=Capsule, length=4.1985, radius=1.5956, center=<4.2929 0 0>, rotation=<rot 90 0 0>}]
 dyn_topHair_e_01_l [{shapeType=Capsule, length=2.0093, radius=2.0221, center=<3.4733 -0.0001 -0.5686>, rotation=<rot 69.9997 0 0>}]
 head [
  {shapeType=Capsule, length=4.75, radius=6, center=<3.03 3.251 0.2817>, rotation=<rot 90 0 0>}
  {shapeType=Capsule, length=0, radius=8.75, center=<5.6291 0.7449 1.0676>, rotation=<rot 90 0 0>}
 ]
 spine_05 [{shapeType=Capsule, length=12.75, radius=9.75, center=<5.9334 -1.5248 0>}]
}
constraints [
 {bone1Name=dyn_topHair_a_01_l, bone2Name=head, swing1LimitDegrees=10, swing2LimitDegrees=20, twistLimitDegrees=2}
 {bone1Name=dyn_topHair_a_02_l, bone2Name=dyn_topHair_a_01_l, swing1LimitDegrees=10, swing2LimitDegrees=20, twistLimitDegrees=2}
 {bone1Name=dyn_topHair_b_01_l, bone2Name=head, swing1LimitDegrees=10, swing2LimitDegrees=20, twistLimitDegrees=2}
 {bone1Name=dyn_topHair_b_02_l, bone2Name=dyn_topHair_b_01_l, swing1LimitDegrees=10, swing2LimitDegrees=20, twistLimitDegrees=2}
 {bone1Name=dyn_topHair_c_01_l, bone2Name=head, swing1LimitDegrees=10, swing2LimitDegrees=20, twistLimitDegrees=2}
 {bone1Name=dyn_topHair_c_02_l, bone2Name=dyn_topHair_c_01_l, swing1LimitDegrees=10, swing2LimitDegrees=20, twistLimitDegrees=2}
 {bone1Name=dyn_topHair_d_01_l, bone2Name=head, swing1LimitDegrees=10, swing2LimitDegrees=20, twistLimitDegrees=2}
 {bone1Name=dyn_topHair_e_01_l, bone2Name=head, swing1LimitDegrees=10, swing2LimitDegrees=20, twistLimitDegrees=2}
 {bone1Name=dyn_topHair_a_01_r, bone2Name=head, swing1LimitDegrees=10, swing2LimitDegrees=20, twistLimitDegrees=2}
 {bone1Name=dyn_topHair_a_02_r, bone2Name=dyn_topHair_a_01_r, swing1LimitDegrees=10, swing2LimitDegrees=20, twistLimitDegrees=2}
 {bone1Name=dyn_topHair_b_01_r, bone2Name=head, swing1LimitDegrees=10, swing2LimitDegrees=20, twistLimitDegrees=2}
 {bone1Name=dyn_topHair_b_02_r, bone2Name=dyn_topHair_b_01_r, swing1LimitDegrees=10, swing2LimitDegrees=20, twistLimitDegrees=2}
 {bone1Name=dyn_topHair_c_01_r, bone2Name=head, swing1LimitDegrees=10, swing2LimitDegrees=20, twistLimitDegrees=2}
 {bone1Name=dyn_topHair_c_02_r, bone2Name=dyn_topHair_c_01_r, swing1LimitDegrees=10, swing2LimitDegrees=20, twistLimitDegrees=2}
 {bone1Name=dyn_hair_a_01_r, bone2Name=head, swing1LimitDegrees=10, swing2LimitDegrees=20, twistLimitDegrees=2}
 {bone1Name=dyn_hair_a_02_r, bone2Name=dyn_hair_a_01_r, swing1LimitDegrees=10, swing2LimitDegrees=20, twistLimitDegrees=2}
 {bone1Name=dyn_hair_a_03_r, bone2Name=dyn_hair_a_02_r, swing1LimitDegrees=10, swing2LimitDegrees=20, twistLimitDegrees=2}
 {bone1Name=dyn_hair_b_01_r, bone2Name=head, swing1LimitDegrees=10, swing2LimitDegrees=20, twistLimitDegrees=2}
 {bone1Name=dyn_hair_b_02_r, bone2Name=dyn_hair_b_01_r, swing1LimitDegrees=10, swing2LimitDegrees=20, twistLimitDegrees=2}
 {bone1Name=dyn_hair_b_03_r, bone2Name=dyn_hair_b_02_r, swing1LimitDegrees=10, swing2LimitDegrees=20, twistLimitDegrees=2}
 {bone1Name=dyn_hair_c_01_r, bone2Name=head, swing1LimitDegrees=10, swing2LimitDegrees=20, twistLimitDegrees=2}
 {bone1Name=dyn_hair_c_02_r, bone2Name=dyn_hair_c_01_r, swing1LimitDegrees=10, swing2LimitDegrees=20, twistLimitDegrees=2}
 {bone1Name=dyn_hair_a_01_l, bone2Name=head, swing1LimitDegrees=10, swing2LimitDegrees=20, twistLimitDegrees=2}
 {bone1Name=dyn_hair_a_02_l, bone2Name=dyn_hair_a_01_l, swing1LimitDegrees=10, swing2LimitDegrees=20, twistLimitDegrees=2}
 {bone1Name=dyn_hair_a_03_l, bone2Name=dyn_hair_a_02_l, swing1LimitDegrees=10, swing2LimitDegrees=20, twistLimitDegrees=2}
 {bone1Name=dyn_hair_b_01_l, bone2Name=head, swing1LimitDegrees=10, swing2LimitDegrees=20, twistLimitDegrees=2}
 {bone1Name=dyn_hair_b_02_l, bone2Name=dyn_hair_b_01_l, swing1LimitDegrees=10, swing2LimitDegrees=20, twistLimitDegrees=2}
 {bone1Name=dyn_hair_c_01_l, bone2Name=head, swing1LimitDegrees=10, swing2LimitDegrees=20, twistLimitDegrees=2}
 {bone1Name=dyn_hair_01_bk, bone2Name=head, swing1LimitDegrees=10, swing2LimitDegrees=35, twistLimitDegrees=2}
 {bone1Name=dyn_hair_02_bk, bone2Name=head, swing1LimitDegrees=10, swing2LimitDegrees=35, twistLimitDegrees=2}
 {bone1Name=dyn_hair_03_bk, bone2Name=head, swing1LimitDegrees=10, swing2LimitDegrees=35, twistLimitDegrees=2}
 {bone1Name=dyn_hair_04_bk, bone2Name=head, swing1LimitDegrees=10, swing2LimitDegrees=20, twistLimitDegrees=2}
 {bone1Name=dyn_hair_05_bk, bone2Name=head, swing1LimitDegrees=10, swing2LimitDegrees=20, twistLimitDegrees=2}
 {
  bone1Name=head
  bone2Name=spine_05
  swing1LimitDegrees=45
  swing1Motion=Free
  swing2LimitDegrees=45
  swing2Motion=Free
  twistLimitDegrees=45
  twistMotion=Free
 }
]
